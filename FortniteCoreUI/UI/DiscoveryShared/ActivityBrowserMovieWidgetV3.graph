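# Discovery activity browser video widget built on FortVideoWidgetBase, layering the video image, subtitles and a throbber in an overlay.
# SetVideoMaterialInstance applies a dynamic copy of a supplied material to the video image.
# PlayOutro plays or jumps to the end of a short animation of that material's thumbnail opacity, zoom and triangle distance parameters, then plays an outro whoosh.
# Initialization plays an intro whoosh.
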WidgetBlueprint /FortniteCoreUI/UI/DiscoveryShared/ActivityBrowserMovieWidgetV3
animations [VideoSwapActivityOutro {
  binding_count=1
  length_seconds=0.24
  object=&self:VideoSwapActivityOutro
  sequence {bindings [Image_VideoTexture {
     id=#1
     possessed_class=Image
     tracks [{
       type=MovieSceneWidgetMaterialTrack
       sections [{
         type=MovieSceneParameterSection
         end_frame=4
         start_frame=0
         channels [
          Thumbnail Opacity {type=MovieSceneScriptingFloatChannel, keys [[0,1],[4.8,0]]}
          Thumbnail Zoom {type=MovieSceneScriptingFloatChannel, keys [[0,0],[4.8,0.05]]}
          Triangle Distance {type=MovieSceneScriptingFloatChannel, keys [[0,0],[4.8,0.304]]}
         ]
        }]
      }]
    }]}
 }]
graphs [
 EventGraph {nodes [
   n1 = Get VideoSwapActivityOutro() -> (VideoSwapActivityOutro:Widget Animation Object Reference)
   PlayAnimation(
    InAnimation=@n1
    StartAtTime=SelectFloat(
     A=0
     B=GetEndTime(self=@n1)
     bPickA=PlayOutro!() -> (OutputDelegate, then, PlayAnimation:bool=true) {is=title K2Node_CustomEvent}.PlayAnimation
    )
   )
   > SpawnSound2D!(
    execute
    Sound:Sound Base Object Reference=Get OutroWhooshSound() -> (OutroWhooshSound:Sound Base Object Reference)
    VolumeMultiplier:float=1
    PitchMultiplier:float=1
    StartTime:float=0
    ConcurrencySettings:Sound Concurrency Object Reference
    bPersistAcrossLevelTransition:bool=false
    bAutoDestroy:bool=true
   ) -> (then, ReturnValue:Audio Component Object Reference)
   Event OnInitialized()
   > SpawnSound2D(
    Sound=Get IntroWhooshSound() -> (IntroWhooshSound:Sound Base Object Reference)
    VolumeMultiplier=1
    PitchMultiplier=1
    StartTime=0
    bPersistAcrossLevelTransition=false
    bAutoDestroy=true
   )
   n2 = Get VideoSwapActivityOutro()
   PlayAnimation(InAnimation=@n2, StartAtTime=GetEndTime(self=@n2), PlayMode=Reverse)
   Reset() {is=title K2Node_CustomEvent}
  ]}
 SetVideoMaterialInstance {nodes [
   n1 = SetVideoMaterialInstance!() -> (then, VideoMaterialInstance:Material Instance Object Reference) {is=title K2Node_FunctionEntry}
   > n2 = CreateDynamicMaterialInstance!(execute, Parent=@n1, OptionalName=None, CreationFlags=None) -> (then, ReturnValue)
   > SetBrushFromMaterial(self=Get Image_VideoTexture() -> (Image_VideoTexture:Image Object Reference), Material=@n2)
  ]}
]
variables [
 IntroWhooshSound {
  type=SoundBase
  default=/Game/Sounds/Fort_UI_Sounds/FrontEnd/MultiProduct/Common/Whooshes/Screens/MSS_UI_MP_Whoosh_Screen_Intro_Experience_Preview
 }
 OutroWhooshSound {
  type=SoundBase
  default=/Game/Sounds/Fort_UI_Sounds/FrontEnd/MultiProduct/Common/Whooshes/Screens/MSS_UI_MP_Whoosh_Screen_Outro_Experience_Preview
 }
]
widgets {
 info {inheritedWidgetCount=4, parentClass=/Script/FortniteUI.FortVideoWidgetBase, rootWidgetClass=&SafeZone, widgetCount=7}
 widgets [
  MainSafeZone {
   widgetClassPath=&SafeZone
   bIsVariable=true
   widget=MainSafeZone
   properties {padBottom=false, padLeft=false, padRight=false, padTop=false}
  }
  SafeZonePadding {
   widgetClassPath=&Border
   bIsVariable=true
   parent=MainSafeZone
   widget=SafeZonePadding
   properties {
    background {
     bIsDynamicallyLoaded=false
     drawAs=NoDrawType
     imageType=NoImage
     mirroring=NoMirror
     resourceName=None
     resourceObject=null
     tiling=NoTile
     imageSize=<32 32>
     margin=<ltrb 0 0 0 0>
     outlineSettings {
      bUseBrushTransparency=false
      roundingType=HalfHeightRadius
      width=0
      color {colorUseRule=UseColor_Specified, specifiedColor=<rgba 0 0 0 0>}
      cornerRadii=<0 0 0 0>
     }
     tintColor {colorUseRule=UseColor_Specified, specifiedColor=<rgba 1 1 1 1>}
     uVRegion {bIsValid=false, max=<0 0>, min=<0 0>}
    }
    brushColor=<rgba 0 0 0 1>
    padding=<ltrb 0 0 0 0>
   }
   slot {class=SafeZoneSlot}
  }
  Overlay_0 {widgetClassPath=&Overlay, parent=SafeZonePadding, widget=Overlay_0, slot {class=BorderSlot, properties {padding=<ltrb 0 0 0 0>}}}
  Image_VideoTexture {
   widgetClassPath=&Image
   bInherited=true
   bIsVariable=true
   parent=Overlay_0
   widget=Image_VideoTexture
   properties {brush {
     bIsDynamicallyLoaded=false
     drawAs=Image
     imageType=NoImage
     mirroring=NoMirror
     resourceName=None
     resourceObject=/FortniteCoreUI/UI/DiscoveryShared/Art/MaterialInstances/MI_UI_Discovery_Carousel_Video
     tiling=NoTile
     imageSize=<32 32>
     margin=<ltrb 0.5 0 0.5 0>
     outlineSettings {
      bUseBrushTransparency=false
      roundingType=HalfHeightRadius
      width=0
      color {colorUseRule=UseColor_Specified, specifiedColor=<rgba 0 0 0 0>}
      cornerRadii=<0 0 0 0>
     }
     tintColor {colorUseRule=UseColor_Specified, specifiedColor=<rgba 1 1 1 1>}
     uVRegion {bIsValid=false, max=<0 0>, min=<0 0>}
    }}
   slot {class=OverlaySlot, properties {horizontalAlignment=HAlign_Fill, verticalAlignment=VAlign_Fill}}
  }
  Subtitles {
   widgetClassPath=/Script/SubtitlesWidgets.SubtitleDisplay
   bInherited=true
   bIsVariable=true
   parent=Overlay_0
   widget=Subtitles
   properties {
    options=/Game/UI/SubtitleOptions
    previewText=SEE-Bot: SEE-Bot SEES THE INSIDE OF THE BAG. THERE IS SO MUCH BAG TO SEE.
    visibility=HitTestInvisible
    wrapTextAt=1170
   }
   slot {class=OverlaySlot, properties {horizontalAlignment=HAlign_Center, verticalAlignment=VAlign_Bottom, padding=<ltrb 0 0 0 100>}}
  }
  Image_Throbber {
   widgetClassPath=&Image
   bInherited=true
   bIsVariable=true
   parent=Overlay_0
   widget=Image_Throbber
   properties {
    visibility=SelfHitTestInvisible
    brush {
     bIsDynamicallyLoaded=false
     drawAs=Image
     imageType=FullColor
     mirroring=NoMirror
     resourceName=None
     resourceObject=/FortUILibrary/UIKit/Throbber/MI_UIKit_Monochrome_Throbber
     tiling=NoTile
     imageSize=<64 64>
     margin=<ltrb 0 0 0 0>
     outlineSettings {
      bUseBrushTransparency=false
      roundingType=HalfHeightRadius
      width=0
      color {colorUseRule=UseColor_Specified, specifiedColor=<rgba 0 0 0 0>}
      cornerRadii=<0 0 0 0>
     }
     tintColor {colorUseRule=UseColor_Specified, specifiedColor=<rgba 1 1 1 1>}
     uVRegion {bIsValid=false, max=<0 0>, min=<0 0>}
    }
   }
   slot {class=OverlaySlot, properties {horizontalAlignment=HAlign_Center, verticalAlignment=VAlign_Center}}
  }
  Button_Skip {widgetClassPath=&CommonButtonBase, bInherited=true, widget=null}
 ]
}
