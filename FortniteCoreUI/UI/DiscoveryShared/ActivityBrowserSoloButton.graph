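# Button widget for the Discovery activity browser, with an input-action icon, main text and optional secondary text over a material-driven backing.
# Hover and Pressed animations drive material scalars on backing and main text, plus float variables whose setters update secondary text.
# Input-method changes reapply mobile size, font, padding and material overrides, optionally hiding the backing on gamepad.
# Disabled state suppresses animations; gamepad clicks play sound; select/deselect still print debug strings.
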WidgetBlueprint /FortniteCoreUI/UI/DiscoveryShared/ActivityBrowserSoloButton
animations [
 Hover {
  binding_count=3
  length_seconds=0.16
  object=&self:Hover
  sequence {
   bindings [
    ActivityBrowserSoloButton_C_26 {
     id=#1
     possessed_class=ActivityBrowserSoloButton_C
     tracks [{
       type=MovieSceneFloatTrack
       display_name=Hover Animate Var
       sections [{type=MovieSceneFloatSection, channels [None {type=MovieSceneScriptingFloatChannel, keys [[0,0],[16,1]]}]}]
      }]
    }
    Text_Content {
     id=#2
     possessed_class=CommonTextBlock
     tracks [{
       type=MovieSceneWidgetMaterialTrack
       display_name=Font.FontMaterial
       sections [{type=MovieSceneParameterSection, channels [Hover_Animate {type=MovieSceneScriptingFloatChannel, keys [[0,0],[16,1]]}]}]
      }]
    }
    ButtonMaterialBorder {
     id=#3
     possessed_class=Border
     tracks [{
       type=MovieSceneWidgetMaterialTrack
       sections [{type=MovieSceneParameterSection, channels [Hover_Animate {type=MovieSceneScriptingFloatChannel, keys [[0,0],[16,1]]}]}]
      }]
    }
   ]
   display_rate {numerator=100}
  }
 }
 Pressed {
  binding_count=3
  length_seconds=0.16
  object=&self:Pressed
  sequence {
   bindings [
    Text_Content {
     id=#4
     possessed_class=CommonTextBlock
     tracks [{
       type=MovieSceneWidgetMaterialTrack
       display_name=Font.FontMaterial
       sections [{type=MovieSceneParameterSection, channels [Pressed_Animate {type=MovieSceneScriptingFloatChannel, keys [[0.0017,0],[16,1]]}]}]
      }]
    }
    ButtonMaterialBorder {
     id=#5
     possessed_class=Border
     tracks [{
       type=MovieSceneWidgetMaterialTrack
       sections [{type=MovieSceneParameterSection, channels [Pressed_Animate {type=MovieSceneScriptingFloatChannel, keys [[0.0017,0],[16,1]]}]}]
      }]
    }
    ActivityBrowserSoloButton_C_27 {
     id=#6
     possessed_class=ActivityBrowserSoloButton_C
     tracks [{
       type=MovieSceneFloatTrack
       display_name=Pressed Animate Var
       sections [{type=MovieSceneFloatSection, channels [None {type=MovieSceneScriptingFloatChannel, keys [[0.0017,0],[16,1]]}]}]
      }]
    }
   ]
   display_rate {numerator=100}
  }
 }
]
graphs [
 EventGraph {nodes [
   n1 = EventOnInputMethodChanged!() -> (OutputDelegate, then, bNewInputType:ECommonInputType Enum) {is=title K2Node_CustomEvent}
   > UpdateStylingOnInputType!(execute, self, Index:ECommonInputType Enum=@n1.bNewInputType) -> (then)
   n2 = OnVisibilityChange!() -> (OutputDelegate, then, InVisibility:ESlateVisibility Enum=Visible) {is=title K2Node_CustomEvent}
   > n3 = Switch on ESlateVisibility!(execute, Selection:ESlateVisibility Enum=@n2.InVisibility) -> (Visible:Exec, Collapsed:Exec, Hidden:Exec, HitTestInvisible:Exec, SelfHitTestInvisible:Exec) {is=title K2Node_SwitchEnum}
   Set ShouldPlayReversed(execute=@n3.Collapsed, ShouldPlayReversed:bool=false) -> (Output_Get:bool)
   > ResetMaterials()
   n4 = Event PreConstruct()
   > Branch(Condition=@n4.IsDesignTime)
   n5 = Get Text_Secondary() -> (Text_Secondary:Common Text Object Reference)
   n6 = Get Text(self:Text Object Reference=@n5)
   > n7 = SetVisibility(self=@n5, InVisibility=Select(Option 0=SelfHitTestInvisible, Option 1=Collapsed, Index=TextIsEmpty(InText=@n6)))
   SetTextStyle(execute=[@n8.then, @n7.then])
   > SetButtonMaterial()
   > ResetFontMaterial()
   > UpdateStylingOnInputType(Index=GetCurrentInputType(self=Get CommonInputSubsystem!() -> (ReturnValue:Common Input Object Reference) {is=title K2Node_GetBlueprintContextEx}))
   Event Construct()
   > Parent: Construct()
   > Bind Event to On Input Method Changed(self=Get CommonInputSubsystem() {is=title K2Node_GetSubsystem}, Delegate=Create Event())
   > n8 = Bind Event to On Visibility Changed(self:User Widget Object Reference, Delegate=@n2) {is=title K2Node_AddDelegate}
   Event On Clicked()
   > n9 = Sequence()
   > n10 = Branch(Condition=Get isDisabled() -> (isDisabled:bool))
   Set ShouldPlayReversed(execute=@n10.else, ShouldPlayReversed=true)
   n11 = Get Pressed() -> (Pressed:Widget Animation Object Reference)
   > n12 = "Play Animation Time Range with Finished event (legacy)"!(
    execute:Exec
    Widget:User Widget Object Reference=Self-Reference()
    InAnimation:Widget Animation Object Reference=@n11
    StartAtTime:float=GetAnimationCurrentTime!(self:User Widget Object Reference, InAnimation:Widget Animation Object Reference=@n11) -> (ReturnValue:float)
    EndAtTime:float=0
    NumLoopsToPlay:int=1
    PlayMode:EUMGSequencePlayMode Enum
    PlaybackSpeed:float=1
   ) -> (then:Exec, Result:UMGSequence Player Object Reference, Finished:Exec) {is=title K2Node_PlayAnimationTimeRange}
   Branch(execute=@n12.Finished, Condition=Get ShouldPlayReversed() -> (ShouldPlayReversed:bool))
   > Set ShouldPlayReversed(ShouldPlayReversed=false)
   > n13 = Branch(Condition=Get isDisabled())
   PlayAnimationReverse(execute=@n13.else, InAnimation=Get Pressed())
   Branch(
    execute=@n9.then_1
    Condition=IsUsingGamepad!(self:Fort Player Controller Object Reference=GetFirstLocalFortPlayerController!(bIncludeNonSpectators:bool=true, bIncludeSpectators:bool=false) -> (ReturnValue:Fort Player Controller Object Reference)) -> (ReturnValue:bool)
   )
   n14 = Get!() -> (PressedSlateSoundOverride_ResourceObject:Object Reference)
   > n15 = Is Valid(InputObject=@n14)
   > n16 = Cast To SoundBase!(execute, Object=@n14) -> (then, CastFailed, AsSound Base:Sound Base Object Reference) {is=title K2Node_DynamicCast}
   > PlaySound2D(Sound=@n16)
   PlaySound2D(execute=@n15.Is Not Valid, Sound=/Game/Sounds/Fort_UI_Sounds/FrontEnd/New_FrontEnd/Fortnite_UI_MainMenu_Generic_01_Cue)
   Event Destruct()
   > Parent: Destruct()
   > Unbind Event from On Input Method Changed(self=Get CommonInputSubsystem(), Delegate=Create Event())
   > Unbind all Events from On Visibility Changed!(execute:Exec, self:User Widget Object Reference) -> (then:Exec) {is=title K2Node_ClearDelegate}
   Event On Disabled() {is=title K2Node_Event}
   > Set isDisabled(isDisabled:bool=true) -> (Output_Get:bool)
   n17 = Get ButtonMaterialBorder() -> (ButtonMaterialBorder:Border Object Reference)
   > n18 = GetDynamicMaterial(self:Border Object Reference=@n17)
   n19 = Get DisabledParamName() -> (DisabledParamName:Name=None)
   > SetScalarParameterValue(self=@n18, ParameterName=@n19, Value=1)
   > ResetMaterials()
   Event On Enabled() {is=title K2Node_Event}
   > Set isDisabled(isDisabled=false)
   > n20 = GetDynamicMaterial(self=@n17)
   > SetScalarParameterValue(self=@n20, ParameterName=@n19, Value=0)
   Event On Hovered()
   n21 = Get isDisabled()
   > n22 = Branch(Condition=@n21)
   n23 = Get Hover() -> (Hover:Widget Animation Object Reference)
   PlayAnimationForward(execute=@n22.else, InAnimation=@n23)
   Event On Unhovered()
   > n24 = Branch(Condition=@n21)
   PlayAnimationReverse(execute=@n24.else, InAnimation=@n23)
   Event On Selected()
   n25 = Get isDisabled()
   > n26 = Branch(Condition=@n25)
   n27 = Get Pressed()
   PlayAnimationForward(execute=@n26.else, InAnimation=@n27)
   > PrintString!(execute, InString=PRESSED, bPrintToScreen, bPrintToLog, TextColor, Duration, Key) -> (then)
   Event On Deselected()
   > n28 = Branch(Condition=@n25)
   PlayAnimationReverse(execute=@n28.else, InAnimation=@n27)
   > PrintString(InString=DEPRESSED)
  ]}
 MakeButtonBrush {nodes [
   n1 = MakeButtonBrush!() -> (then, ResourceObject:Object Reference) {is=title K2Node_FunctionEntry}
   n2 = Break Slate Brush!(SlateBrush:"Slate Brush Structure (by ref)"=Get Background!(self:Border Object Reference=Get ButtonMaterialBorder()) -> (Background:Slate Brush Structure)) -> (
    TintColor:Slate Color Structure
    DrawAs:ESlateBrushDrawType Enum
    Tiling:ESlateBrushTileType Enum
    Mirroring:ESlateBrushMirrorType Enum
    ImageSize:Vector 2D Structure
    Margin:Margin Structure
    ResourceObject:Object Reference
    OutlineSettings:Slate Brush Outline Settings Structure
   ) {is=title K2Node_BreakStruct}
   n3 = Get ButtonBoxScaleSizeX() -> (ButtonBoxScaleSizeX:double)
   n4 = Get ButtonBoxScaleSizeY() -> (ButtonBoxScaleSizeY:double)
   > Return Node!(
    execute
    SlateBrush:Slate Brush Structure=Make Slate Brush!(
     TintColor:Slate Color Structure=@n2
     DrawAs:ESlateBrushDrawType Enum=@n2.DrawAs
     Tiling:ESlateBrushTileType Enum=@n2.Tiling
     Mirroring:ESlateBrushMirrorType Enum=@n2.Mirroring
     ImageSize:Vector 2D Structure=MakeVector2D(X=@n3, Y=@n4)
     Margin:Margin Structure=@n2.Margin
     ResourceObject:Object Reference=@n1
     OutlineSettings:Slate Brush Outline Settings Structure=(CornerRadii=(X=0,Y=0,Z=0,W=0),Color=(SpecifiedColor=(R=0,G=0,B=0,A=0),ColorUseRule=UseColor_Specified),Width=0,RoundingType=HalfHeightRadius,bUseBrushTransparency=False)
    ) -> (SlateBrush:Slate Brush Structure) {is=title K2Node_MakeStruct}
   )
  ]}
 ResetFontMaterial {nodes [
   ResetFontMaterial() {is=title K2Node_FunctionEntry}
   > n1 = GetDynamicFontMaterial!(execute, self:Text Object Reference=Get Text_Content() -> (Text_Content:Common Text Object Reference)) -> (then, ReturnValue:Material Instance Dynamic Object Reference)
   > SetScalarParameterValue(self=@n1, ParameterName=Hover_Animate, Value=0)
   > SetScalarParameterValue(self=@n1, ParameterName=Pressed_Animate, Value=0)
  ]}
 ResetMaterials {nodes [
   ResetMaterials() {is=title K2Node_FunctionEntry}
   > StopAnimation(InAnimation=Get Pressed())
   n1 = Get Hover()
   > PlayAnimation(InAnimation=@n1, StartAtTime=GetEndTime(self=@n1), PlayMode=Reverse)
   > ResetFontMaterial() {is=title K2Node_CallFunction}
   > n2 = GetDynamicMaterial(self=Get ButtonMaterialBorder())
   > SetScalarParameterValue(self=@n2, ParameterName=Hover_Animate, Value=0)
   > SetScalarParameterValue(self=@n2, ParameterName=Pressed_Animate, Value=0)
  ]}
 SetButtonMaterial {nodes [
   SetButtonMaterial() {is=title K2Node_FunctionEntry}
   n1 = IsUsingTouch(PlayerController=GetOwningPlayer())
   n2 = IsValid(Object=Get MaterialOverride() -> (MaterialOverride:Material Interface Object Reference))
   n3 = Get bEnableMobileOverrides() -> (bEnableMobileOverrides:bool)
   > n4 = MakeButtonBrush!(
    execute
    self
    ResourceObject:Object Reference=Select(
     Option 0:Object Reference=Get ButtonMaterial() -> (ButtonMaterial:Material Interface Object Reference)
     Option 1:Object Reference=Get MaterialOverride()
     Index=AND Boolean!(A=@n1, B=@n2, C:bool=@n3) -> (ReturnValue)
    ) -> (ReturnValue:Object Reference)
   ) -> (then, SlateBrush:Slate Brush Structure) {is=title K2Node_CallFunction}
   > SetBrush(self:Border Object Reference=Get ButtonMaterialBorder(), InBrush=@n4)
   > n5 = GetDynamicMaterial(self=Get ButtonMaterialBorder())
   > SetScalarParameterValue(
    self=@n5
    ParameterName=Get SharpnessUParamName() -> (SharpnessUParamName:Name=None)
    Value=Get ButtonSharpnessX() -> (ButtonSharpnessX:double)
   )
   > SetScalarParameterValue(
    self=@n5
    ParameterName=Get SharpnessVParamName() -> (SharpnessVParamName:Name=None)
    Value=Get ButtonSharpnessY() -> (ButtonSharpnessY:double)
   )
   > SetScalarParameterValue(
    self=@n5
    ParameterName=Get UseBoxScaleUParamName() -> (UseBoxScaleUParamName:Name=None)
    Value=Get UseBoxScalingX() -> (UseBoxScalingX:double)
   )
   > SetScalarParameterValue(
    self=@n5
    ParameterName=Get UseBoxScaleVParamName() -> (UseBoxScaleVParamName:Name=None)
    Value=Get UseBoxScalingY() -> (UseBoxScalingY:double)
   )
  ]}
 SetHoverAnimateVar {nodes [
   n1 = SetHoverAnimateVar!() -> (then, NewValue:double=~) {is=title K2Node_FunctionEntry}
   > Set HoverAnimateVar(HoverAnimateVar:double=@n1) -> (Output_Get:double)
   > n2 = GetDynamicFontMaterial(self=Get Text_Secondary())
   > SetScalarParameterValue(self=@n2, ParameterName=Hover_Animate, Value=Get HoverAnimateVar() -> (HoverAnimateVar:double))
  ]}
 SetPressedAnimateVar {nodes [
   n1 = SetPressedAnimateVar!() -> (then, NewValue:double=~) {is=title K2Node_FunctionEntry}
   > Set PressedAnimateVar(PressedAnimateVar:double=@n1) -> (Output_Get:double)
   > n2 = GetDynamicFontMaterial(self=Get Text_Secondary())
   > SetScalarParameterValue(self=@n2, ParameterName=Pressed_Animate, Value=Get PressedAnimateVar() -> (PressedAnimateVar:double))
  ]}
 SetText {nodes [
   n1 = SetText!() -> (then, Text:Text) {is=title K2Node_FunctionEntry}
   n2 = Get Text_Content()
   > "SetText (Text)"(self=@n2, InText=@n1)
   > SetRenderShear!(
    execute
    self:Widget Object Reference=@n2
    Shear:Vector 2D Structure=MakeVector2D(X=Get TextShearX() -> (TextShearX:double), Y=Get TextShearY() -> (TextShearY:double))
   ) -> (then)
  ]}
 SetTextStyle {nodes [
   SetTextStyle() {is=title K2Node_FunctionEntry}
   n1 = IsUsingTouch(PlayerController=GetOwningPlayer())
   n2 = Get bEnableMobileOverrides()
   > SetPadding!(
    execute
    self:Border Object Reference=Get DynamicPaddingBorder() -> (DynamicPaddingBorder:Border Object Reference)
    InPadding:Margin Structure=Select(
     Option 0:Margin Structure=Get TextPadding() -> (TextPadding:Margin Structure)
     Option 1:Margin Structure=Get TextPaddingOverride() -> (TextPaddingOverride:Margin Structure)
     Index=AND Boolean!(A=@n1, B=@n2) -> (ReturnValue)
    ) -> (ReturnValue:Margin Structure)
   ) -> (then)
  ]}
 Toggle Mobile Overrides {nodes [
   n1 = Toggle Mobile Overrides!() -> (then, bApplyMobileOverrides:bool=~) {is=title K2Node_FunctionEntry}
   > n2 = Branch(Condition=@n1)
   > SetMinDimensions!(
    execute
    self:Common Button Base Object Reference
    InMinWidth:int=Get MinWidthOverride() -> (MinWidthOverride:int=0)
    InMinHeight:int=Get MinHeightOverride() -> (MinHeightOverride:int=0)
   ) -> (then)
   n3 = Get Text_Content()
   n4 = Break Slate Font Info!(SlateFontInfo:"Slate Font Info Structure (by ref)"=Get Font!(self:Text Object Reference=@n3) -> (Font:Slate Font Info Structure)) -> (
    FontObject:Object Reference
    FontMaterial:Object Reference
    OutlineSettings:Font Outline Settings Structure
    TypefaceFontName:Name
    Size:float=~
    LetterSpacing:int
    SkewAmount:float=~
    bForceMonospaced:bool=~
    bMaterialIsStencil:bool=~
    MonospacedWidth:float=~
   ) {is=title K2Node_BreakStruct}
   n5 = Get FontSizeOverride() -> (FontSizeOverride:int=0)
   > n6 = SetFont(
    self=@n3
    InFontInfo=MakeSlateFontInfo!(
     FontObject:Object Reference=@n4
     FontMaterial:Object Reference=@n4.FontMaterial
     OutlineSettings:Font Outline Settings Structure=@n4.OutlineSettings
     TypefaceFontName:Name=@n4.TypefaceFontName
     Size:float="To Float (Integer)"(InInt=@n5)
     LetterSpacing:int=@n4.LetterSpacing
     SkewAmount:float=0
     bForceMonospaced:bool=false
     bMaterialIsStencil:bool=false
     MonospacedWidth:float=1
    ) -> (ReturnValue:Slate Font Info Structure)
   )
   SetButtonMaterial(execute=[@n7.then, @n6.then]) {is=title K2Node_CallFunction}
   > SetTextStyle() {is=title K2Node_CallFunction}
   SetMinDimensions(execute=@n2.else, InMinWidth=Get MinWidth() -> (MinWidth:int=0), InMinHeight=Get MinHeight() -> (MinHeight:int=0))
   > n7 = SetFont(self=Get Text_Content(), InFontInfo=Get Font!() -> (Font))
  ]}
 UpdateStylingOnInputType {nodes [
   n1 = UpdateStylingOnInputType!() -> (then, Index:ECommonInputType Enum) {is=title K2Node_FunctionEntry}
   > Toggle Mobile Overrides!(execute, self, bApplyMobileOverrides:bool=AND Boolean(A=IsMobileGame(), B=Get bEnableMobileOverrides())) -> (then) {is=title K2Node_CallFunction}
   > n2 = Switch on ECommonInputType!(execute, Selection:ECommonInputType Enum=@n1) -> (MouseAndKeyboard:Exec, Gamepad:Exec, Touch:Exec, Count:Exec) {is=title K2Node_SwitchEnum}
   n3 = Branch(execute=@n2.Gamepad, Condition=Get HideButtonBackingOnGamepad() -> (HideButtonBackingOnGamepad:bool))
   > n4 = GetDynamicMaterial(self=Get ButtonMaterialBorder())
   > SetScalarParameterValue(self=@n4, ParameterName=AlphaOveride, Value=0)
   n5 = GetDynamicMaterial(execute=[@n3.else, @n2.MouseAndKeyboard, @n2.Touch], self=Get ButtonMaterialBorder())
   > SetScalarParameterValue(self=@n5, ParameterName=AlphaOveride, Value=1)
  ]}
]
variables [
 ButtonBoxScaleSizeX {type=number, default=64}
 ButtonBoxScaleSizeY {type=number, default=64}
 ButtonMaterial {type=MaterialInterface, default=/Game/UI/DiscoveryShared/Art/MI_UI_Discovery_Button}
 ButtonSharpnessX {type=number, default=64}
 ButtonSharpnessY {type=number, default=64}
 DisabledParamName {type=string, default=bIsDisabled}
 HideButtonBackingOnGamepad {type=boolean}
 HoverAnimateVar {type=number}
 isDisabled {type=boolean}
 PressedAnimateVar {type=number}
 SharpnessUParamName {type=string, default=Sharpness_U}
 SharpnessVParamName {type=string, default=Sharpness_V}
 ShouldPlayReversed {type=boolean}
 TextPadding {type=Margin, default=<ltrb 32 0 32 0>}
 TextShearX {type=number}
 TextShearY {type=number}
 UseBoxScaleUParamName {type=string, default=UseBoxScaleU}
 UseBoxScaleVParamName {type=string, default=UseBoxScaleV}
 UseBoxScalingX {type=number}
 UseBoxScalingY {type=number, default=1}
]
widgets {
 info {inheritedWidgetCount=3, parentClass=/Script/FortniteUI.FortActivityBrowserSoloButton, rootWidgetClass=&SizeBox, widgetCount=8}
 widgets [
  MinSizesSB {
   widgetClassPath=&SizeBox
   bIsVariable=true
   widget=MinSizesSB
   properties {bOverride_MinDesiredHeight=true, bOverride_MinDesiredWidth=true, minDesiredHeight=48, minDesiredWidth=128}
  }
  ButtonMaterialBorder {
   widgetClassPath=&Border
   bIsVariable=true
   parent=MinSizesSB
   widget=ButtonMaterialBorder
   properties {
    background {
     bIsDynamicallyLoaded=false
     drawAs=Box
     imageType=NoImage
     mirroring=NoMirror
     resourceName=None
     resourceObject=&MI_UI_Discovery_Button
     tiling=NoTile
     imageSize=<64 64>
     margin=<ltrb 0.5 0.5 0.5 0.5>
     outlineSettings {
      bUseBrushTransparency=false
      roundingType=HalfHeightRadius
      width=0
      color {colorUseRule=UseColor_Specified, specifiedColor=<rgba 0 0 0 0>}
      cornerRadii=<0 0 0 0>
     }
     tintColor {colorUseRule=UseColor_Specified, specifiedColor=<rgba 1 1 1 1>}
     uVRegion {bIsValid=false, max=<0 0>, min=<0 0>}
    }
    padding=<ltrb 0 0 0 0>
   }
   slot {class=SizeBoxSlot}
  }
  DynamicPaddingBorder {
   widgetClassPath=&Border
   bIsVariable=true
   parent=ButtonMaterialBorder
   widget=DynamicPaddingBorder
   properties {
    background {
     bIsDynamicallyLoaded=false
     drawAs=NoDrawType
     imageType=NoImage
     mirroring=NoMirror
     resourceName=None
     resourceObject=null
     tiling=NoTile
     imageSize=<32 32>
     margin=<ltrb 0 0 0 0>
     outlineSettings {
      bUseBrushTransparency=false
      roundingType=HalfHeightRadius
      width=0
      color {colorUseRule=UseColor_Specified, specifiedColor=<rgba 0 0 0 0>}
      cornerRadii=<0 0 0 0>
     }
     tintColor {colorUseRule=UseColor_Specified, specifiedColor=<rgba 1 1 1 1>}
     uVRegion {bIsValid=false, max=<0 0>, min=<0 0>}
    }
    padding=<ltrb 16 0 16 0>
   }
   slot {class=BorderSlot, properties {padding=<ltrb 0 0 0 0>}}
  }
  ContentHB {
   widgetClassPath=&HorizontalBox
   parent=DynamicPaddingBorder
   widget=ContentHB
   slot {class=BorderSlot, properties {padding=<ltrb 16 0 16 0>}}
  }
  InputActionWidget {
   widgetClassPath=&CommonActionWidget
   bInherited=true
   bIsVariable=true
   parent=ContentHB
   widget=InputActionWidget
   properties {
    progressMaterialParam=Percentage
    progressMaterialBrush {
     bIsDynamicallyLoaded=false
     drawAs=Image
     imageType=NoImage
     mirroring=NoMirror
     resourceName=None
     resourceObject=/Game/UI/Materials/HollowCircularProgressBar
     tiling=NoTile
     imageSize=<36 36>
     margin=<ltrb 0 0 0 0>
     outlineSettings {
      bUseBrushTransparency=false
      roundingType=HalfHeightRadius
      width=0
      color {colorUseRule=UseColor_Specified, specifiedColor=<rgba 0 0 0 0>}
      cornerRadii=<0 0 0 0>
     }
     tintColor {colorUseRule=UseColor_Specified, specifiedColor=<rgba 1 1 0 1>}
     uVRegion {bIsValid=false, max=<0 0>, min=<0 0>}
    }
   }
   slot {class=HorizontalBoxSlot, properties {padding=<ltrb 0 0 12 0>}}
  }
  VerticalBox_TextContainer {
   widgetClassPath=&VerticalBox
   parent=ContentHB
   widget=VerticalBox_TextContainer
   slot {class=HorizontalBoxSlot, properties {verticalAlignment=VAlign_Center, padding=<ltrb 0 10 0 10>, size {sizeRule=Fill, value=1}}}
  }
  Text_Content {
   widgetClassPath=&CommonTextBlock
   bInherited=true
   bIsVariable=true
   parent=VerticalBox_TextContainer
   widget=Text_Content
   properties {
    justification=Center
    scrollStyle=&TextScrollStyle-Base_C
    text="{Text}"
    textTransformPolicy=ToUpper
    font {
     bForceMonospaced=false
     bMaterialIsStencil=false
     fontMaterial=/FortniteCoreUI/UI/DiscoveryShared/Art/MI_UI_Discovery_ButtonText
     fontObject=&BurbankSmall
     letterSpacing=0
     monospacedWidth=1
     size=20
     skewAmount=0
     typefaceFontName=BLACK
     outlineSettings {
      bApplyOutlineToDropShadows=false
      bMiteredCorners=false
      bSeparateFillAlpha=false
      outlineMaterial=null
      outlineSize=0
      outlineColor=<rgba 0 0 0 1>
     }
    }
    shadowOffset=<0 0>
   }
   slot {class=VerticalBoxSlot, properties {verticalAlignment=VAlign_Center}}
  }
  Text_Secondary {
   widgetClassPath=&CommonTextBlock
   bInherited=true
   bIsVariable=true
   parent=VerticalBox_TextContainer
   widget=Text_Secondary
   properties {
    bAutoCollapseWithEmptyText=true
    justification=Center
    scrollStyle=&TextScrollStyle-Base_C
    font {
     bForceMonospaced=false
     bMaterialIsStencil=false
     fontMaterial=&MI_UI_Discovery_ButtonText
     fontObject=&BurbankSmall
     letterSpacing=0
     monospacedWidth=1
     size=20
     skewAmount=0
     typefaceFontName=BLACK
     outlineSettings {
      bApplyOutlineToDropShadows=false
      bMiteredCorners=false
      bSeparateFillAlpha=false
      outlineMaterial=null
      outlineSize=0
      outlineColor=<rgba 0 0 0 1>
     }
    }
    shadowOffset=<0 0>
   }
   slot {class=VerticalBoxSlot, properties {verticalAlignment=VAlign_Center}}
  }
 ]
}
