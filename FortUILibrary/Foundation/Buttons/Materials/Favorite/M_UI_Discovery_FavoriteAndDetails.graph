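# Translucent UI material for the Discovery favorite button icon, tinting a mask texture by active/inactive and hover states.
# The mask is scaled about its center on hover and smoothstepped with adjustable edge and margin for sharpening.
# Its channels separate inner and outer regions, each with state-driven color and inactive opacity, while green gives overall coverage.
# An IsSimpleMask switch bypasses this, using green to blend two active colors and red as opacity.
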Material /FortUILibrary/Foundation/Buttons/Materials/Favorite/M_UI_Discovery_FavoriteAndDetails
graph {
 properties {
  bUsedWithStaticMesh=false
  blendMode=BLEND_Translucent
  materialDomain=MD_UI
  pixelDepthOffsetMode=PDOM_Legacy
  usedShadingModels=Default Lit
 }
 nodes [
  n1 = ScalarParameter() {group=States, parameterName=bOnHover_RGBA, sliderMax=1}
  LinearInterpolate(
   A=ScalarParameter() {defaultValue=1, group=Alpha, parameterName=Active_Inner, sliderMax=1}
   B=ScalarParameter() {defaultValue=1, group=Alpha, parameterName=Active_Inner__Hovered, sliderMax=1}
   Alpha=@n1
  )
  LinearInterpolate(
   A=ScalarParameter() {defaultValue=1, group=Alpha, parameterName=Active_Outer, sliderMax=1}
   B=ScalarParameter() {defaultValue=1, group=Alpha, parameterName=Active_Outer_Hovered, sliderMax=1}
   Alpha=@n1
  )
  n2 = ScalarParameter() {defaultValue=1, group=Shape, parameterName=Scale_Start}
  n3 = ScalarParameter() {defaultValue=1, group=Shape, parameterName=Scale_Finish}
  n4 = ScalarParameter() {group=States, parameterName=bOnHover_Scale, sliderMax=1}
  n5 = TextureObjectParameter() {
   parameterName=Texture
   samplerType=SAMPLERTYPE_Masks
   texture=/FortUILibrary/Foundation/Buttons/Materials/Favorite/Textures/T_UI_Discovery_Favorite
  }
  n6 = TextureSample(
   UVs=MaterialFunctionCall(UVs=TextureCoordinate(), Texture Scale=LinearInterpolate(A=@n2, B=@n3, Alpha=@n4)) {materialFunction=&ScaleUVsByCenter}.UVs
   Tex=@n5
  ) {samplerSource=SSM_Clamp_WorldGroupSettings}
  n7 = ScalarParameter() {group=States, parameterName=bOnHover_RGBA, sliderMax=1}
  n8 = LinearInterpolate(
   A=VectorParameter() {group=Colors, parameterName=OuterRGB_Active, defaultValue=<rgba 1 0.3185 0.5841 1>}.RGB
   B=VectorParameter() {group=Colors, parameterName=OuterRGB_Active_Hovered, defaultValue=<rgba 1 0.6667 0.7966 1>}.RGB
   Alpha=@n7
  )
  n9 = LinearInterpolate(
   A=VectorParameter() {group=Colors, parameterName=OuterRGB_Inactive, defaultValue=<rgba 1 1 1 1>}.RGB
   B=VectorParameter() {group=Colors, parameterName=OuterRGB_Inactive_Hovered, defaultValue=<rgba 1 1 1 1>}.RGB
   Alpha=@n7
  )
  n10 = ScalarParameter() {group=States, parameterName=bIsInactive, sliderMax=1}
  n11 = LinearInterpolate(
   A=VectorParameter() {group=Colors, parameterName=InnerRGB_Active, defaultValue=<rgba 0.7835 0.0086 0.1845 1>}.RGB
   B=VectorParameter() {group=Colors, parameterName=InnerRGB_Active_Hovered, defaultValue=<rgba 0.7835 0.0979 0.2536 1>}.RGB
   Alpha=@n7
  )
  n12 = LinearInterpolate(
   A=VectorParameter() {group=Colors, parameterName=InnerRGB_Inactive, defaultValue=<rgba 1 1 1 1>}.RGB
   B=VectorParameter() {group=Colors, parameterName=InnerRGB_Inactive_Hovered, defaultValue=<rgba 1 1 1 1>}.RGB
   Alpha=@n7
  )
  n13 = ScalarParameter() {group=States, parameterName=bOnHover_Sharpen, sliderMax=1}
  n14 = LinearInterpolate(
   A=ScalarParameter() {defaultValue=0.1875, group=Shape, parameterName=SSEdge_Start, sliderMax=1}
   B=ScalarParameter() {defaultValue=0.25, group=Shape, parameterName=SSEdge_Finish, sliderMax=1}
   Alpha=@n13
  )
  n15 = LinearInterpolate(
   A=ScalarParameter() {defaultValue=0.125, group=Shape, parameterName=SSMargin_Start, sliderMax=1}
   B=ScalarParameter() {defaultValue=0.1, group=Shape, parameterName=SSMargin_Finish, sliderMax=1}
   Alpha=@n13
  )
  n16 = SmoothStep(Min=Add(A=@n14, B=Multiply(A=@n15) {constB=-1}), Max=Add(A=@n14, B=@n15), Value=@n6.RGB)
  n17 = ComponentMask(None=@n16) {b=true}
  n18 = ComponentMask(None=@n16) {r=true}
  n19 = LinearInterpolate(B=OneMinus(None=Saturate(None=@n17)), Alpha=OneMinus(None=Saturate(None=@n18)))
  StaticSwitchParameter(
   True=LinearInterpolate(
    A=VectorParameter() {group=Colors, parameterName=OuterRGB_Active, defaultValue=<rgba 1 0.3185 0.5841 1>}.RGB
    B=VectorParameter() {group=Colors, parameterName=InnerRGB_Active, defaultValue=<rgba 0.7835 0.0086 0.1845 1>}.RGB
    Alpha=Saturate(None=@n6.G)
   )
   False=LinearInterpolate(A=LinearInterpolate(A=@n8, B=@n9, Alpha=@n10), B=LinearInterpolate(A=@n11, B=@n12, Alpha=@n10), Alpha=@n19)
  ) {group=States, parameterName=IsSimpleMask}
  n20 = Constant() {r=1}
  n21 = ScalarParameter() {group=States, parameterName=bIsInactive, sliderMax=1}
  n22 = LinearInterpolate(
   A=@n20
   B=LinearInterpolate(
    A=ScalarParameter() {defaultValue=0.6, group=Alpha, parameterName=Inactive_Outer, sliderMax=1}
    B=ScalarParameter() {defaultValue=0.6, group=Alpha, parameterName=Inactive_Outer_Hovered, sliderMax=1}
    Alpha=@n1
   )
   Alpha=@n21
  )
  n23 = LinearInterpolate(
   A=@n20
   B=LinearInterpolate(
    A=ScalarParameter() {defaultValue=0.3, group=Alpha, parameterName=Inactive_Inner, sliderMax=1}
    B=ScalarParameter() {defaultValue=0.3, group=Alpha, parameterName=Inactive_Inner_Hovered, sliderMax=1}
    Alpha=@n1
   )
   Alpha=@n21
  )
  n24 = ComponentMask(None=@n16) {g=true}
  StaticSwitchParameter(True=Saturate(None=@n6.R), False=LinearInterpolate(B=LinearInterpolate(A=@n22, B=@n23, Alpha=@n19), Alpha=Saturate(None=@n24))) {group=States, parameterName=IsSimpleMask}
 ]
}
