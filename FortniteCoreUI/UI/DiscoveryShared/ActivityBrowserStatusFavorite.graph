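# Favorite status tag for the Discovery activity browser, a CommonUserWidget with a material-driven heart icon and "Favorite" label.
# OnFavorited pops the heart in, then slides and fades in the pink label; OnUnfavorited greys and hides both.
# PlayFavorite and PlayUnfavorite set IsFavorited, stop the opposing animation and play their own.
# SetFavoriteNoAnimation snaps to either state by playing the matching animation from its end time.
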WidgetBlueprint /FortniteCoreUI/UI/DiscoveryShared/ActivityBrowserStatusFavorite
animations [
 OnFavorited {
  binding_count=3
  length_seconds=0.8
  object=&self:OnFavorited
  sequence {
   bindings [
    SBHeart {
     id=#1
     possessed_class=SizeBox
     tracks [
      {
       type=MovieScene2DTransformTrack
       sections [{
         type=MovieScene2DTransformSection
         channels [
          Translation.X {type=MovieSceneScriptingFloatChannel}
          Translation.Y {type=MovieSceneScriptingFloatChannel}
          Angle {type=MovieSceneScriptingFloatChannel}
          Scale.X {type=MovieSceneScriptingFloatChannel, keys [[0,0.75],[24,1.375],[48.0033,1]]}
          Scale.Y {type=MovieSceneScriptingFloatChannel, keys [[0,0.75],[24,1.375],[48.0033,1]]}
          Shear.X {type=MovieSceneScriptingFloatChannel}
          Shear.Y {type=MovieSceneScriptingFloatChannel}
         ]
        }]
      }
      {
       type=MovieSceneFloatTrack
       display_name=Render Opacity
       sections [{type=MovieSceneFloatSection, channels [None {type=MovieSceneScriptingFloatChannel, keys [[0,0],[48.0033,1]]}]}]
      }
     ]
    }
    Image_StatusIcon {
     id=#2
     possessed_class=Image
     tracks [{
       type=MovieSceneWidgetMaterialTrack
       sections [{type=MovieSceneParameterSection, channels [bIsInactive {type=MovieSceneScriptingFloatChannel, const [0,1]}]}]
      }]
    }
    Text_TagName {
     id=#3
     possessed_class=CommonTextBlock
     tracks [
      {
       type=MovieSceneFloatTrack
       display_name=Render Opacity
       sections [{type=MovieSceneFloatSection, channels [None {type=MovieSceneScriptingFloatChannel, keys [[48.0033,0],[80.0017,1]]}]}]
      }
      {
       type=MovieScene2DTransformTrack
       sections [{
         type=MovieScene2DTransformSection
         channels [
          Translation.X {type=MovieSceneScriptingFloatChannel, keys [[48.0017,-32],[80.0017,0]]}
          Translation.Y {type=MovieSceneScriptingFloatChannel}
          Angle {type=MovieSceneScriptingFloatChannel}
          Scale.X {type=MovieSceneScriptingFloatChannel}
          Scale.Y {type=MovieSceneScriptingFloatChannel}
          Shear.X {type=MovieSceneScriptingFloatChannel}
          Shear.Y {type=MovieSceneScriptingFloatChannel}
         ]
        }]
      }
      {
       type=MovieSceneColorTrack
       display_name=Color and Opacity
       sections [{
         type=MovieSceneColorSection
         channels [
          Color.R {type=MovieSceneScriptingFloatChannel, const [1,1]}
          Color.G {type=MovieSceneScriptingFloatChannel, const [0.0469,1]}
          Color.B {type=MovieSceneScriptingFloatChannel, const [0.2135,1]}
          Color.A {type=MovieSceneScriptingFloatChannel, const [1,1]}
         ]
        }]
      }
     ]
    }
   ]
   display_rate {numerator=100}
  }
 }
 OnUnfavorited {
  binding_count=3
  length_seconds=0.8
  object=&self:OnUnfavorited
  sequence {
   bindings [
    SBHeart {
     id=#4
     possessed_class=SizeBox
     tracks [
      {
       type=MovieSceneFloatTrack
       display_name=Render Opacity
       sections [{type=MovieSceneFloatSection, channels [None {type=MovieSceneScriptingFloatChannel, keys [[16,1],[40,0]]}]}]
      }
      {
       type=MovieScene2DTransformTrack
       sections [{
         type=MovieScene2DTransformSection
         channels [
          Translation.X {type=MovieSceneScriptingFloatChannel}
          Translation.Y {type=MovieSceneScriptingFloatChannel}
          Angle {type=MovieSceneScriptingFloatChannel}
          Scale.X {type=MovieSceneScriptingFloatChannel, keys [[0,1],[16,1.25],[40,0.75]]}
          Scale.Y {type=MovieSceneScriptingFloatChannel, keys [[0,1],[16,1.25],[40,0.75]]}
          Shear.X {type=MovieSceneScriptingFloatChannel}
          Shear.Y {type=MovieSceneScriptingFloatChannel}
         ]
        }]
      }
     ]
    }
    Text_TagName {
     id=#5
     possessed_class=CommonTextBlock
     tracks [
      {
       type=MovieSceneColorTrack
       display_name=Color and Opacity
       sections [{
         type=MovieSceneColorSection
         channels [
          Color.R {type=MovieSceneScriptingFloatChannel, const [1,1]}
          Color.G {type=MovieSceneScriptingFloatChannel, keys [[0,0.0469],[24,1]]}
          Color.B {type=MovieSceneScriptingFloatChannel, keys [[0,0.2135],[24,1]]}
          Color.A {type=MovieSceneScriptingFloatChannel, keys [[0,1],[24,0.6]]}
         ]
        }]
      }
      {
       type=MovieScene2DTransformTrack
       sections [{
         type=MovieScene2DTransformSection
         channels [
          Translation.X {type=MovieSceneScriptingFloatChannel, keys [[40,0],[80,-32]]}
          Translation.Y {type=MovieSceneScriptingFloatChannel}
          Angle {type=MovieSceneScriptingFloatChannel}
          Scale.X {type=MovieSceneScriptingFloatChannel}
          Scale.Y {type=MovieSceneScriptingFloatChannel}
          Shear.X {type=MovieSceneScriptingFloatChannel}
          Shear.Y {type=MovieSceneScriptingFloatChannel}
         ]
        }]
      }
      {
       type=MovieSceneFloatTrack
       display_name=Render Opacity
       sections [{type=MovieSceneFloatSection, channels [None {type=MovieSceneScriptingFloatChannel, keys [[40,1],[80,0]]}]}]
      }
     ]
    }
    Image_StatusIcon {
     id=#6
     possessed_class=Image
     tracks [{
       type=MovieSceneWidgetMaterialTrack
       sections [{type=MovieSceneParameterSection, channels [bIsInactive {type=MovieSceneScriptingFloatChannel, keys [[8,0],[24,1]]}]}]
      }]
    }
   ]
   display_rate {numerator=100}
  }
 }
]
graphs [EventGraph {nodes [
   PlayUnfavorite() {is=title K2Node_CustomEvent}
   > Set IsFavorited(IsFavorited:bool=false) -> (Output_Get:bool)
   > Check and Stop Animation!(execute:Exec, Animation:Widget Animation Object Reference=Get OnFavorited() -> (OnFavorited:Widget Animation Object Reference)) -> (then:Exec) {is=title K2Node_MacroInstance}
   > PlayAnimationForward(InAnimation=Get OnUnfavorited() -> (OnUnfavorited:Widget Animation Object Reference))
   PlayFavorite() {is=title K2Node_CustomEvent}
   > Set IsFavorited(IsFavorited=true)
   > Check and Stop Animation(Animation=Get OnUnfavorited())
   > PlayAnimationForward(InAnimation=Get OnFavorited())
   n1 = SetFavoriteNoAnimation!() -> (OutputDelegate, then, IsFavorited:bool=~) {is=title K2Node_CustomEvent}
   > n2 = Set IsFavorited(IsFavorited=@n1.IsFavorited)
   > n3 = Branch(Condition=@n2)
   > Check and Stop Animation(Animation=Get OnUnfavorited())
   n4 = Get OnFavorited()
   > PlayAnimation(InAnimation=@n4, StartAtTime=GetEndTime(self=@n4))
   Check and Stop Animation(execute=@n3.else, Animation=Get OnFavorited())
   n5 = Get OnUnfavorited()
   > PlayAnimation(InAnimation=@n5, StartAtTime=GetEndTime(self=@n5))
  ]}]
variables [IsFavorited {type=boolean}]
widgets {
 info {inheritedWidgetCount=0, parentClass=&CommonUserWidget, rootWidgetClass=&HorizontalBox, widgetCount=4}
 widgets [
  TagHB {widgetClassPath=&HorizontalBox, widget=TagHB}
  SBHeart {
   widgetClassPath=&SizeBox
   parent=TagHB
   widget=SBHeart
   properties {bOverride_HeightOverride=true, bOverride_WidthOverride=true, heightOverride=32, widthOverride=32}
   slot {class=HorizontalBoxSlot, properties {padding=<ltrb 0 0 8 0>}}
  }
  Image_StatusIcon {
   widgetClassPath=&Image
   parent=SBHeart
   widget=Image_StatusIcon
   properties {
    visibility=HitTestInvisible
    brush {
     bIsDynamicallyLoaded=false
     drawAs=Image
     imageType=NoImage
     mirroring=NoMirror
     resourceName=None
     resourceObject=/FortUILibrary/Foundation/Buttons/Materials/Favorite/M_UI_Discovery_FavoriteAndDetails
     tiling=NoTile
     imageSize=<32 32>
     margin=<ltrb 0 0 0 0>
     outlineSettings {
      bUseBrushTransparency=false
      roundingType=HalfHeightRadius
      width=0
      color {colorUseRule=UseColor_Specified, specifiedColor=<rgba 0 0 0 0>}
      cornerRadii=<0 0 0 0>
     }
     tintColor {colorUseRule=UseColor_Specified, specifiedColor=<rgba 1 1 1 1>}
     uVRegion {bIsValid=false, max=<0 0>, min=<0 0>}
    }
   }
   slot {class=SizeBoxSlot}
  }
  Text_TagName {
   widgetClassPath=&CommonTextBlock
   parent=TagHB
   widget=Text_TagName
   properties {
    mobileFontSizeMultiplier=1.2
    text=Favorite
    textTransformPolicy=ToUpper
    colorAndOpacity {colorUseRule=UseColor_Specified, specifiedColor=<rgba 1 0.0469 0.2135 1>}
    font {
     bForceMonospaced=false
     bMaterialIsStencil=false
     fontMaterial=null
     fontObject=&BurbankSmall
     letterSpacing=0
     monospacedWidth=1
     size=15
     skewAmount=0.24
     typefaceFontName=Black
     outlineSettings {
      bApplyOutlineToDropShadows=false
      bMiteredCorners=false
      bSeparateFillAlpha=false
      outlineMaterial=null
      outlineSize=0
      outlineColor=<rgba 0 0 0 1>
     }
    }
    margin=<ltrb 0 -4 0 -4>
    shadowOffset=<0 0>
   }
   slot {class=HorizontalBoxSlot, properties {verticalAlignment=VAlign_Center}}
  }
 ]
}
