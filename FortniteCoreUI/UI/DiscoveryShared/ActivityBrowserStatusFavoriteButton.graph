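# CommonUI favorite button for the Discovery activity browser, with an icon, pulse image, "Favorite" label and input-action glyph.
# OnIsFavoriteChanged stores the state and plays the favorited or unfavorited animation after stopping its opposite, while SetFavoriteNoAnim snaps material, text color and scale to the state.
# Hover plays one animation forward and reverse.
# OnClicked only branches on favorite state and gamepad use, with nothing connected afterward.
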WidgetBlueprint /FortniteCoreUI/UI/DiscoveryShared/ActivityBrowserStatusFavoriteButton
animations [
 OnPCHoverUnHover {
  binding_count=1
  length_seconds=0.12
  object=&self:OnPCHoverUnHover
  sequence {
   bindings [Image_FavoriteIcon {
     id=#1
     possessed_class=Image
     tracks [
      {
       type=MovieScene2DTransformTrack
       sections [{
         type=MovieScene2DTransformSection
         channels [
          Scale.X {type=MovieSceneScriptingFloatChannel, keys [[0,1],[12,1.125]]}
          Scale.Y {type=MovieSceneScriptingFloatChannel, keys [[0,1],[12,1.125]]}
         ]
        }]
      }
      {
       type=MovieSceneWidgetMaterialTrack
       sections [{type=MovieSceneParameterSection, channels [bOnHover_RGBA {type=MovieSceneScriptingFloatChannel, keys [[0,0],[12,1]]}]}]
      }
     ]
    }]
   display_rate {numerator=100}
  }
 }
 OnFavorited {
  binding_count=4
  length_seconds=0.48
  object=&self:OnFavorited
  sequence {
   bindings [
    Image_FavoriteIcon {
     id=#2
     possessed_class=Image
     tracks [{
       type=MovieSceneWidgetMaterialTrack
       sections [{type=MovieSceneParameterSection, channels [bIsInactive {type=MovieSceneScriptingFloatChannel, keys [[0,1],[20,0]]}]}]
      }]
    }
    Image_Pulse {
     id=#3
     possessed_class=Image
     tracks [
      {
       type=MovieScene2DTransformTrack
       sections [{
         type=MovieScene2DTransformSection
         channels [
          Translation.X {type=MovieSceneScriptingFloatChannel}
          Translation.Y {type=MovieSceneScriptingFloatChannel, keys [[10,1.5],[38,3]]}
          Scale.X {type=MovieSceneScriptingFloatChannel, keys [[10,1],[38,2.5]]}
          Scale.Y {type=MovieSceneScriptingFloatChannel, keys [[10,1],[38,2.5]]}
         ]
        }]
      }
      {
       type=MovieSceneFloatTrack
       display_name=Render Opacity
       sections [{type=MovieSceneFloatSection, channels [None {type=MovieSceneScriptingFloatChannel, keys [[5,0],[20,1],[28,1],[48,0]]}]}]
      }
      {
       type=MovieSceneByteTrack
       display_name=Visibility
       sections [{type=MovieSceneByteSection, channels [None {type=MovieSceneScriptingByteChannel, keys [[10,3],[48,1]]}]}]
      }
     ]
    }
    Text_TagName {
     id=#4
     possessed_class=CommonTextBlock
     tracks [{
       type=MovieSceneColorTrack
       display_name=Color and Opacity
       sections [{
         type=MovieSceneColorSection
         channels [
          Color.R {type=MovieSceneScriptingFloatChannel, const [1,2]}
          Color.G {type=MovieSceneScriptingFloatChannel, keys [[0,1],[16,0.0469]]}
          Color.B {type=MovieSceneScriptingFloatChannel, keys [[0,1],[16,0.2135]]}
          Color.A {type=MovieSceneScriptingFloatChannel, keys [[0,0.6],[16,1]]}
         ]
        }]
      }]
    }
    ScalingSB {
     id=#5
     possessed_class=SizeBox
     tracks [{
       type=MovieScene2DTransformTrack
       sections [{
         type=MovieScene2DTransformSection
         channels [
          Scale.X {type=MovieSceneScriptingFloatChannel, keys [[0,1],[8,0.875],[16,1.175],[32,1]]}
          Scale.Y {type=MovieSceneScriptingFloatChannel, keys [[0,1],[8,0.875],[16,1.175],[32,1]]}
         ]
        }]
      }]
    }
   ]
   display_rate {numerator=100}
  }
 }
 OnUnfavorited {
  binding_count=2
  length_seconds=0.24
  object=&self:OnUnfavorited
  sequence {
   bindings [
    Image_FavoriteIcon {
     id=#2
     possessed_class=Image
     tracks [
      {
       type=MovieSceneWidgetMaterialTrack
       sections [{type=MovieSceneParameterSection, channels [bIsInactive {type=MovieSceneScriptingFloatChannel, keys [[0,0],[24,1]]}]}]
      }
      {
       type=MovieScene2DTransformTrack
       sections [{
         type=MovieScene2DTransformSection
         channels [
          Scale.X {type=MovieSceneScriptingFloatChannel, keys [[0,1],[12,0.875],[24,1]]}
          Scale.Y {type=MovieSceneScriptingFloatChannel, keys [[0,1],[12,0.875],[24,1]]}
         ]
        }]
      }
     ]
    }
    Text_TagName {
     id=#6
     possessed_class=CommonTextBlock
     tracks [{
       type=MovieSceneColorTrack
       display_name=Color and Opacity
       sections [{
         type=MovieSceneColorSection
         channels [
          Color.R {type=MovieSceneScriptingFloatChannel, const [1,1]}
          Color.G {type=MovieSceneScriptingFloatChannel, keys [[0,0.0469],[24,1]]}
          Color.B {type=MovieSceneScriptingFloatChannel, keys [[0,0.2135],[24,1]]}
          Color.A {type=MovieSceneScriptingFloatChannel, keys [[0,1],[24,0.6]]}
         ]
        }]
      }]
    }
   ]
   display_rate {numerator=100}
  }
 }
]
graphs [EventGraph {nodes [
   n1 = OnIsFavoriteChanged!() -> (OutputDelegate, then, bIsFavorite:bool=~) {is=title K2Node_CustomEvent}
   > n2 = Set Is Favorite(Is Favorite:bool=@n1.bIsFavorite) -> (Output_Get:bool)
   > n3 = Branch(Condition=@n2)
   > Check and Stop Animation!(execute:Exec, Animation:Widget Animation Object Reference=Get OnUnfavorited() -> (OnUnfavorited:Widget Animation Object Reference)) -> (then:Exec) {is=title K2Node_MacroInstance}
   > PlayAnimationForward(InAnimation=Get OnFavorited() -> (OnFavorited:Widget Animation Object Reference))
   Check and Stop Animation(execute=@n3.else, Animation=Get OnFavorited())
   > PlayAnimationForward(InAnimation=Get OnUnfavorited())
   n4 = SetFavoriteNoAnim!() -> (OutputDelegate, then, IsFavorite:bool=~) {is=title K2Node_CustomEvent}
   > Set Is Favorite(Is Favorite=@n4.IsFavorite)
   > SetVisibility(self=Get Image_Pulse() -> (Image_Pulse:Image Object Reference), InVisibility=Collapsed)
   n5 = Get ScalingSB() -> (ScalingSB:Size Box Object Reference)
   n6 = Get!(self:Widget Object Reference=@n5) -> (
    RenderTransform_Translation:Vector 2D Structure
    RenderTransform_Scale:Vector 2D Structure
    RenderTransform_Shear:Vector 2D Structure
    RenderTransform_Angle:float=~
   )
   > SetRenderTransform!(
    execute
    self:Widget Object Reference=@n5
    InTransform:Widget Transform Structure=Make Widget Transform!(
     Translation:Vector 2D Structure=@n6
     Scale:Vector 2D Structure=(X=1,Y=1)
     Shear:Vector 2D Structure=@n6.RenderTransform_Shear
     Angle:float=@n6.RenderTransform_Angle
    ) -> (WidgetTransform:Widget Transform Structure) {is=title K2Node_MakeStruct}
   ) -> (then)
   > n7 = Branch(Condition=Get Is Favorite() -> (Is Favorite:bool))
   > Check and Stop Animation(Animation=Get OnUnfavorited())
   n8 = Get Image_FavoriteIcon() -> (Image_FavoriteIcon:Image Object Reference)
   > n9 = GetDynamicMaterial(self=@n8)
   n10 = Get bIsInactiveParam() -> (bIsInactiveParam:Name=None)
   > SetScalarParameterValue(self=@n9, ParameterName=@n10, Value=0)
   n11 = Get Text_TagName() -> (Text_TagName:Common Text Object Reference)
   > SetColorAndOpacity(self=@n11, InColorAndOpacity=Make Slate Color(SpecifiedColor=(R=1,G=0.0469,B=0.2135,A=1)))
   Check and Stop Animation(execute=@n7.else, Animation=Get OnFavorited())
   > n12 = GetDynamicMaterial(self=@n8)
   > SetScalarParameterValue(self=@n12, ParameterName=@n10, Value=1)
   > SetColorAndOpacity(self=@n11, InColorAndOpacity=Make Slate Color(SpecifiedColor=(R=1,G=1,B=1,A=0.6)))
   Event On Clicked()
   > n13 = Branch(Condition=Get Is Favorite())
   n14 = IsUsingGamepad!(self:Fort Player Controller Object Reference=GetFirstLocalFortPlayerController!(bIncludeNonSpectators:bool=true, bIncludeSpectators:bool=false) -> (ReturnValue:Fort Player Controller Object Reference)) -> (ReturnValue:bool)
   > Branch(Condition=@n14)
   Branch(execute=@n13.else, Condition=@n14)
   Event On Hovered()
   n15 = Get OnPCHoverUnHover() -> (OnPCHoverUnHover:Widget Animation Object Reference)
   > PlayAnimationForward(InAnimation=@n15)
   Event On Unhovered()
   > PlayAnimationReverse(InAnimation=@n15)
   Event PreConstruct()
   > n16 = GetDynamicMaterial(self=Get Image_FavoriteIcon())
   > SetScalarParameterValue(self=@n16, ParameterName=Get bOnHoverSharpenParam() -> (bOnHoverSharpenParam:Name=None), Value=1)
   > SetScalarParameterValue(self=@n16, ParameterName=Get bOnHoverScaleParam() -> (bOnHoverScaleParam:Name=None), Value=1)
  ]}]
variables [
 bIsInactiveParam {type=string, default=bIsInactive}
 bOnHoverScaleParam {type=string, default=bOnHover_Scale}
 bOnHoverSharpenParam {type=string, default=bOnHover_Sharpen}
 Is Favorite {type=boolean}
]
widgets {
 info {inheritedWidgetCount=1, parentClass=&CommonButtonBase, rootWidgetClass=&HorizontalBox, widgetCount=8}
 widgets [
  HorizontalBox_0 {widgetClassPath=&HorizontalBox, widget=HorizontalBox_0}
  InputActionWidget {
   widgetClassPath=&CommonActionWidget
   bInherited=true
   bIsVariable=true
   parent=HorizontalBox_0
   widget=InputActionWidget
   properties {
    visibility=SelfHitTestInvisible
    iconRimBrush {
     bIsDynamicallyLoaded=false
     drawAs=Image
     imageType=NoImage
     mirroring=NoMirror
     resourceName=None
     resourceObject=/FortniteCoreUI/UI/DiscoveryShared/Art/MI_UI_Circle_WhiteKeybindBorder_SemiTransparent
     tiling=NoTile
     imageSize=<38 38>
     margin=<ltrb 0 0 0 0>
     outlineSettings {
      bUseBrushTransparency=false
      roundingType=HalfHeightRadius
      width=0
      color {colorUseRule=UseColor_Specified, specifiedColor=<rgba 0 0 0 0>}
      cornerRadii=<0 0 0 0>
     }
     tintColor {colorUseRule=UseColor_Specified, specifiedColor=<rgba 0 0 0 1>}
     uVRegion {bIsValid=false, max=<0 0>, min=<0 0>}
    }
   }
   slot {class=HorizontalBoxSlot, properties {horizontalAlignment=HAlign_Center, verticalAlignment=VAlign_Center}}
  }
  Overlay_0 {
   widgetClassPath=&Overlay
   parent=HorizontalBox_0
   widget=Overlay_0
   slot {class=HorizontalBoxSlot, properties {verticalAlignment=VAlign_Center, padding=<ltrb 2 0 8 0>}}
  }
  ScalingSB {
   widgetClassPath=&SizeBox
   bIsVariable=true
   parent=Overlay_0
   widget=ScalingSB
   properties {bOverride_HeightOverride=true, bOverride_WidthOverride=true, heightOverride=32, widthOverride=32}
   slot {class=OverlaySlot, properties {horizontalAlignment=HAlign_Fill, verticalAlignment=VAlign_Fill}}
  }
  Overlay_1 {widgetClassPath=&Overlay, parent=ScalingSB, widget=Overlay_1, slot {class=SizeBoxSlot}}
  Image_Pulse {
   widgetClassPath=&Image
   bIsVariable=true
   parent=Overlay_1
   widget=Image_Pulse
   properties {
    renderOpacity=0
    visibility=Collapsed
    brush {
     bIsDynamicallyLoaded=false
     drawAs=Image
     imageType=NoImage
     mirroring=NoMirror
     resourceName=None
     resourceObject=/FortUILibrary/Foundation/Buttons/Materials/Favorite/MI_UI_Favorite_Pulse
     tiling=NoTile
     imageSize=<32 32>
     margin=<ltrb 0 0 0 0>
     outlineSettings {
      bUseBrushTransparency=false
      roundingType=HalfHeightRadius
      width=0
      color {colorUseRule=UseColor_Specified, specifiedColor=<rgba 0 0 0 0>}
      cornerRadii=<0 0 0 0>
     }
     tintColor {colorUseRule=UseColor_Specified, specifiedColor=<rgba 1 1 1 1>}
     uVRegion {bIsValid=false, max=<0 0>, min=<0 0>}
    }
   }
   slot {class=OverlaySlot, properties {horizontalAlignment=HAlign_Fill, verticalAlignment=VAlign_Fill}}
  }
  Image_FavoriteIcon {
   widgetClassPath=&Image
   bIsVariable=true
   parent=Overlay_1
   widget=Image_FavoriteIcon
   properties {
    visibility=SelfHitTestInvisible
    brush {
     bIsDynamicallyLoaded=false
     drawAs=Image
     imageType=NoImage
     mirroring=NoMirror
     resourceName=None
     resourceObject=/FortUILibrary/Foundation/Buttons/Materials/Favorite/MI_UI_Favorite_Base
     tiling=NoTile
     imageSize=<32 32>
     margin=<ltrb 0 0 0 0>
     outlineSettings {
      bUseBrushTransparency=false
      roundingType=HalfHeightRadius
      width=0
      color {colorUseRule=UseColor_Specified, specifiedColor=<rgba 0 0 0 0>}
      cornerRadii=<0 0 0 0>
     }
     tintColor {colorUseRule=UseColor_Specified, specifiedColor=<rgba 1 1 1 1>}
     uVRegion {bIsValid=false, max=<0 0>, min=<0 0>}
    }
   }
   slot {class=OverlaySlot, properties {horizontalAlignment=HAlign_Fill, verticalAlignment=VAlign_Fill}}
  }
  Text_TagName {
   widgetClassPath=&CommonTextBlock
   bIsVariable=true
   parent=HorizontalBox_0
   widget=Text_TagName
   properties {
    mobileFontSizeMultiplier=1.2
    text=Favorite
    textTransformPolicy=ToUpper
    colorAndOpacity {colorUseRule=UseColor_Specified, specifiedColor=<rgba 1 0.0469 0.2135 1>}
    font {
     bForceMonospaced=false
     bMaterialIsStencil=false
     fontMaterial=null
     fontObject=&BurbankSmall
     letterSpacing=0
     monospacedWidth=1
     size=15
     skewAmount=0.24
     typefaceFontName=Black
     outlineSettings {
      bApplyOutlineToDropShadows=false
      bMiteredCorners=false
      bSeparateFillAlpha=false
      outlineMaterial=null
      outlineSize=0
      outlineColor=<rgba 0 0 0 1>
     }
    }
    margin=<ltrb 0 -4 0 -4>
    shadowOffset=<0 0>
   }
   slot {class=HorizontalBoxSlot, properties {verticalAlignment=VAlign_Center}}
  }
 ]
}
