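# Privacy button for the Discovery mode-set UI, showing a title, gamepad action icon and Public/Private toggle, built on FortActivityPrivacyButton.
# Desired-privacy changes set the toggle; availability changes broadcast visibility via a dispatcher and enable or dim the button, unless bForceEnabled keeps it enabled.
# Unless disabled, hover, selection and click play Hover and Pressed animations, though Pressed has no tracks.
# Clicks on gamepad also play a sound.
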WidgetBlueprint /FortniteCoreUI/UI/DiscoveryShared/ModeSet/ActivityPrivacyButton
animations [
 Hover {
  binding_count=2
  length_seconds=0.2
  object=&self:Hover
  sequence {
   bindings [
    Image_Bg {
     id=#1
     possessed_class=Image
     tracks [{
       type=MovieSceneWidgetMaterialTrack
       sections [{type=MovieSceneParameterSection, channels [AnimateHighlight {type=MovieSceneScriptingFloatChannel, keys [[0,0],[15,1],[20,1]]}]}]
      }]
    }
    Border_Title {
     id=#2
     possessed_class=Border
     tracks [{
       type=MovieSceneColorTrack
       display_name=Content Color and Opacity
       sections [{
         type=MovieSceneColorSection
         channels [
          Color.R {type=MovieSceneScriptingFloatChannel, keys [[0,1],[15,0]]}
          Color.G {type=MovieSceneScriptingFloatChannel, keys [[0,1],[15,0.004]]}
          Color.B {type=MovieSceneScriptingFloatChannel, keys [[0,1],[15,0.0723]]}
          Color.A {type=MovieSceneScriptingFloatChannel}
         ]
        }]
      }]
    }
   ]
   display_rate {numerator=100}
  }
 }
 Pressed {
  binding_count=0
  length_seconds=0.15
  object=&self:Pressed
  sequence {bindings [], display_rate {numerator=100}}
 }
]
graphs [
 EventGraph {nodes [
   Event PreConstruct()
   > SetSidesToPad!(
    execute
    self:Safe Zone Object Reference=Get SafeZone() -> (SafeZone:Safe Zone Object Reference)
    InPadLeft:bool=Get PadLeft() -> (PadLeft:bool)
    InPadRight:bool=Get PadRight() -> (PadRight:bool)
    InPadTop:bool=false
    InPadBottom:bool=false
   ) -> (then)
   Event On Clicked()
   > n1 = Sequence()
   > n2 = Branch(Condition=Get isDisabled() -> (isDisabled:bool))
   Set ShouldPlayReversed(execute=@n2.else, ShouldPlayReversed:bool=true) -> (Output_Get:bool)
   n3 = Get Pressed() -> (Pressed:Widget Animation Object Reference)
   > n4 = "Play Animation Time Range with Finished event (legacy)"!(
    execute:Exec
    Widget:User Widget Object Reference=Self-Reference()
    InAnimation:Widget Animation Object Reference=@n3
    StartAtTime:float=GetAnimationCurrentTime!(self:User Widget Object Reference, InAnimation:Widget Animation Object Reference=@n3) -> (ReturnValue:float)
    EndAtTime:float=0
    NumLoopsToPlay:int=1
    PlayMode:EUMGSequencePlayMode Enum
    PlaybackSpeed:float=1
   ) -> (then:Exec, Result:UMGSequence Player Object Reference, Finished:Exec) {is=title K2Node_PlayAnimationTimeRange}
   Branch(execute=@n4.Finished, Condition=Get ShouldPlayReversed() -> (ShouldPlayReversed:bool))
   > Set ShouldPlayReversed(ShouldPlayReversed=false)
   > n5 = Branch(Condition=Get isDisabled())
   PlayAnimationReverse(execute=@n5.else, InAnimation=Get Pressed())
   Branch(
    execute=@n1.then_1
    Condition=IsUsingGamepad!(self:Fort Player Controller Object Reference=GetFirstLocalFortPlayerController!(bIncludeNonSpectators:bool=true, bIncludeSpectators:bool=false) -> (ReturnValue:Fort Player Controller Object Reference)) -> (ReturnValue:bool)
   )
   n6 = Get!() -> (PressedSlateSoundOverride_ResourceObject:Object Reference)
   > n7 = Is Valid(InputObject=@n6)
   > n8 = Cast To SoundBase!(execute, Object=@n6) -> (then, CastFailed, AsSound Base:Sound Base Object Reference) {is=title K2Node_DynamicCast}
   > PlaySound2D(Sound=@n8)
   PlaySound2D(execute=@n7.Is Not Valid, Sound=/Game/Sounds/Fort_UI_Sounds/FrontEnd/New_FrontEnd/Fortnite_UI_MainMenu_Generic_01_Cue)
   n9 = Event OnDesiredPrivacySelectionChanged!() -> (OutputDelegate, then, bIsSelected:bool) {is=title K2Node_Event}
   > SetToggleState!(execute, self:Text Toggle Object Reference=Get TextToggle() -> (TextToggle:Text Toggle Object Reference), IsOn:bool=@n9.bIsSelected) -> (then)
   n10 = Event OnPrivacyAvailabilityChanged!() -> (OutputDelegate, then, bIsAvailable:bool, bShouldBeVisibile:bool) {is=title K2Node_Event}
   > Call On Privacy Available Changed!(execute:Exec, self:Activity Privacy Button Object Reference, IsAvailable:bool=@n10.bShouldBeVisibile) -> (then:Exec) {is=title K2Node_CallDelegate}
   > n11 = Branch(Condition=Get bForceEnabled() -> (bForceEnabled:bool))
   Set isDisabled(execute=[@n11.then, @n13.then], isDisabled:bool=false) -> (Output_Get:bool)
   > SetIsEnabled(bInIsEnabled=true)
   n12 = Get Image_Dimmer() -> (Image_Dimmer:Image Object Reference)
   > SetVisibility(self=@n12, InVisibility=Collapsed)
   n13 = Branch(execute=@n11.else, Condition=@n10.bIsAvailable)
   Set isDisabled(execute=@n13.else, isDisabled=true)
   > SetIsEnabled(bInIsEnabled=false)
   > SetVisibility(self=@n12, InVisibility=SelfHitTestInvisible)
   Event On Disabled() {is=title K2Node_Event}
   > Set isDisabled(isDisabled=true)
   Event On Enabled() {is=title K2Node_Event}
   > Set isDisabled(isDisabled=false)
   Event On Hovered()
   n14 = Get isDisabled()
   > n15 = Branch(Condition=@n14)
   n16 = Get Hover() -> (Hover:Widget Animation Object Reference)
   PlayAnimationForward(execute=@n15.else, InAnimation=@n16)
   Event On Unhovered()
   > n17 = Branch(Condition=@n14)
   PlayAnimationReverse(execute=@n17.else, InAnimation=@n16)
   Event On Selected()
   n18 = Get isDisabled()
   > n19 = Branch(Condition=@n18)
   n20 = Get Pressed()
   PlayAnimationForward(execute=@n19.else, InAnimation=@n20)
   Event On Deselected()
   > n21 = Branch(Condition=@n18)
   PlayAnimationReverse(execute=@n21.else, InAnimation=@n20)
  ]}
 OnPrivacyAvailableChanged {nodes [OnPrivacyAvailableChanged!() -> (then, IsAvailable:bool=~) {is=title K2Node_FunctionEntry}]}
]
variables [
 bForceEnabled {type=boolean}
 Button Description {type=string, default="{Text}"}
 ButtonBoxScaleSizeX {type=number, default=64}
 ButtonBoxScaleSizeY {type=number, default=64}
 ButtonMaterial {type=MaterialInterface, default=/Game/UI/DiscoveryShared/Art/MI_UI_Discovery_Button}
 ButtonSharpnessX {type=number, default=64}
 ButtonSharpnessY {type=number, default=64}
 DisabledParamName {type=string, default=bIsDisabled}
 HideButtonBackingOnGamepad {type=boolean}
 isDisabled {type=boolean}
 OnPrivacyAvailableChanged {}
 PadLeft {type=boolean}
 PadRight {type=boolean}
 SharpnessUParamName {type=string, default=Sharpness_U}
 SharpnessVParamName {type=string, default=Sharpness_V}
 ShouldPlayReversed {type=boolean}
 TextPadding {type=Margin, default=<ltrb 32 0 32 0>}
 TextShearX {type=number}
 TextShearY {type=number}
 UseBoxScaleUParamName {type=string, default=UseBoxScaleU}
 UseBoxScaleVParamName {type=string, default=UseBoxScaleV}
 UseBoxScalingX {type=number}
 UseBoxScalingY {type=number, default=1}
]
widgets {
 info {inheritedWidgetCount=1, parentClass=/Script/FortniteUI.FortActivityPrivacyButton, rootWidgetClass=&Overlay, widgetCount=11}
 widgets [
  Overlay_2 {widgetClassPath=&Overlay, widget=Overlay_2}
  TouchOnly_Size {
   widgetClassPath=&CommonHardwareVisibilityBorder
   parent=Overlay_2
   widget=TouchOnly_Size
   properties {visibilityQuery {
     autoDescription=" ANY( Input.Touch )"
     tokenStreamVersion=0
     userDescription=touch-only
     queryTokenStream [0,1,1,1,0]
     tagDictionary [{tagName=Input.Touch}]
    }}
   slot {class=OverlaySlot, properties {horizontalAlignment=HAlign_Center, verticalAlignment=VAlign_Center}}
  }
  touch_size {
   widgetClassPath=&Spacer
   parent=TouchOnly_Size
   widget=touch_size
   properties {size=<0 80>}
   slot {class=BorderSlot, properties {padding=<ltrb 0 0 0 0>}}
  }
  Image_Bg {
   widgetClassPath=&Image
   bIsVariable=true
   parent=Overlay_2
   widget=Image_Bg
   properties {brush {
     bIsDynamicallyLoaded=false
     drawAs=Image
     imageType=NoImage
     mirroring=NoMirror
     resourceName=None
     resourceObject=/Game/UI/DiscoveryShared/ModeSet/Materials/MI_UI_SidebarStyle_Button_CategoryInst
     tiling=NoTile
     imageSize=<32 32>
     margin=<ltrb 0 0 0 0>
     outlineSettings {
      bUseBrushTransparency=false
      roundingType=HalfHeightRadius
      width=0
      color {colorUseRule=UseColor_Specified, specifiedColor=<rgba 0 0 0 0>}
      cornerRadii=<0 0 0 0>
     }
     tintColor {colorUseRule=UseColor_Specified, specifiedColor=<rgba 1 1 1 1>}
     uVRegion {bIsValid=false, max=<0 0>, min=<0 0>}
    }}
   slot {class=OverlaySlot, properties {horizontalAlignment=HAlign_Fill, verticalAlignment=VAlign_Fill}}
  }
  SafeZone {
   widgetClassPath=&SafeZone
   bIsVariable=true
   parent=Overlay_2
   widget=SafeZone
   slot {class=OverlaySlot, properties {horizontalAlignment=HAlign_Fill, verticalAlignment=VAlign_Center, padding=<ltrb 20 10 20 10>}}
  }
  ContentHB {widgetClassPath=&HorizontalBox, bIsVariable=true, parent=SafeZone, widget=ContentHB, slot {class=SafeZoneSlot}}
  Border_Title {
   widgetClassPath=&Border
   parent=ContentHB
   widget=Border_Title
   properties {
    background {
     bIsDynamicallyLoaded=false
     drawAs=NoDrawType
     imageType=NoImage
     mirroring=NoMirror
     resourceName=None
     resourceObject=null
     tiling=NoTile
     imageSize=<32 32>
     margin=<ltrb 0 0 0 0>
     outlineSettings {
      bUseBrushTransparency=false
      roundingType=HalfHeightRadius
      width=0
      color {colorUseRule=UseColor_Specified, specifiedColor=<rgba 0 0 0 0>}
      cornerRadii=<0 0 0 0>
     }
     tintColor {colorUseRule=UseColor_Specified, specifiedColor=<rgba 1 1 1 1>}
     uVRegion {bIsValid=false, max=<0 0>, min=<0 0>}
    }
    padding=<ltrb 0 0 15 0>
   }
   slot {class=HorizontalBoxSlot, properties {horizontalAlignment=HAlign_Left, verticalAlignment=VAlign_Center, size {sizeRule=Fill, value=1}}}
  }
  Text_Title {
   widgetClassPath=&CommonTextBlock
   bIsVariable=true
   parent=Border_Title
   widget=Text_Title
   properties {
    justification=Center
    mobileFontSizeMultiplier=1.3
    scrollStyle=&TextScrollStyle-Base_C
    text=Privacy
    textTransformPolicy=ToUpper
    font {
     bForceMonospaced=false
     bMaterialIsStencil=false
     fontMaterial=null
     fontObject=&BurbankSmall
     letterSpacing=0
     monospacedWidth=1
     size=16
     skewAmount=0
     typefaceFontName=Black
     outlineSettings {
      bApplyOutlineToDropShadows=false
      bMiteredCorners=false
      bSeparateFillAlpha=false
      outlineMaterial=null
      outlineSize=0
      outlineColor=<rgba 0 0 0 1>
     }
    }
    shadowOffset=<0 0>
   }
   slot {class=BorderSlot, properties {padding=<ltrb 0 0 15 0>}}
  }
  InputActionWidget {
   widgetClassPath=&CommonActionWidget
   bInherited=true
   bIsVariable=true
   parent=ContentHB
   widget=InputActionWidget
   properties {
    progressMaterialParam=Percentage
    progressMaterialBrush {
     bIsDynamicallyLoaded=false
     drawAs=Image
     imageType=NoImage
     mirroring=NoMirror
     resourceName=None
     resourceObject=/Game/UI/Materials/HollowCircularProgressBar
     tiling=NoTile
     imageSize=<36 36>
     margin=<ltrb 0 0 0 0>
     outlineSettings {
      bUseBrushTransparency=false
      roundingType=HalfHeightRadius
      width=0
      color {colorUseRule=UseColor_Specified, specifiedColor=<rgba 0 0 0 0>}
      cornerRadii=<0 0 0 0>
     }
     tintColor {colorUseRule=UseColor_Specified, specifiedColor=<rgba 1 1 0 1>}
     uVRegion {bIsValid=false, max=<0 0>, min=<0 0>}
    }
   }
   slot {class=HorizontalBoxSlot, properties {horizontalAlignment=HAlign_Right, verticalAlignment=VAlign_Center, padding=<ltrb 0 0 5 0>}}
  }
  TextToggle {
   widgetClassPath=/Game/UI/DiscoveryShared/Status/TextToggle
   bIsVariable=true
   parent=ContentHB
   widget=TextToggle
   properties {offText=Public, onText=Private}
   slot {class=HorizontalBoxSlot, properties {horizontalAlignment=HAlign_Right, verticalAlignment=VAlign_Center}}
  }
  Image_Dimmer {
   widgetClassPath=&Image
   bIsVariable=true
   parent=Overlay_2
   widget=Image_Dimmer
   properties {visibility=Collapsed, colorAndOpacity=<rgba 0 0.0078 0.0392 0.8>}
   slot {class=OverlaySlot, properties {horizontalAlignment=HAlign_Fill, verticalAlignment=VAlign_Fill}}
  }
 ]
}
