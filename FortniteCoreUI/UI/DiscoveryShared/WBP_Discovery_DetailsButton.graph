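# Details icon button for Discovery tiles, built on CommonButtonBase with an input action keybind glyph and optional drop shadow.
# Tile focus events toggle interaction and play a scale-and-fade animation forward or in reverse, while tile reset disables and collapses it.
# An optional hover animation scales the icon and lightens its material color.
# Construct and design-time PreConstruct apply icon size and keybind placement, offset by render transform (ignoring layout) or slot padding.
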WidgetBlueprint /FortniteCoreUI/UI/DiscoveryShared/WBP_Discovery_DetailsButton
animations [
 OnFocus {
  binding_count=2
  length_seconds=0.16
  object=&self:OnFocus
  sequence {
   bindings [
    Overlay_Details {
     id=#1
     possessed_class=Overlay
     tracks [
      {
       type=MovieSceneFloatTrack
       display_name=Render Opacity
       sections [{type=MovieSceneFloatSection, end_frame=16, start_frame=0, channels [None {type=MovieSceneScriptingFloatChannel, keys [[0,0],[16,1]]}]}]
      }
      {
       type=MovieScene2DTransformTrack
       sections [{
         type=MovieScene2DTransformSection
         end_frame=16
         start_frame=0
         channels [
          Scale.X {type=MovieSceneScriptingFloatChannel, keys [[0,0.5],[16,1]]}
          Scale.Y {type=MovieSceneScriptingFloatChannel, keys [[0,0.5],[16,1]]}
         ]
        }]
      }
     ]
    }
    InputActionWidget {
     id=#2
     possessed_class=CommonActionWidget
     tracks [
      {
       type=MovieSceneFloatTrack
       display_name=Render Opacity
       sections [{type=MovieSceneFloatSection, channels [None {type=MovieSceneScriptingFloatChannel, keys [[2,0],[16,1]]}]}]
      }
      {
       type=MovieScene2DTransformTrack
       sections [{
         type=MovieScene2DTransformSection
         channels [
          Translation.X {type=MovieSceneScriptingFloatChannel}
          Translation.Y {type=MovieSceneScriptingFloatChannel}
          Angle {type=MovieSceneScriptingFloatChannel}
          Scale.X {type=MovieSceneScriptingFloatChannel, keys [[2,0.5],[16,1]]}
          Scale.Y {type=MovieSceneScriptingFloatChannel, keys [[2,0.5],[16,1]]}
          Shear.X {type=MovieSceneScriptingFloatChannel}
          Shear.Y {type=MovieSceneScriptingFloatChannel}
         ]
        }]
      }
     ]
    }
   ]
   display_rate {numerator=100}
  }
 }
 OnHover {
  binding_count=2
  length_seconds=0.16
  object=&self:OnHover
  sequence {
   bindings [
    Overlay_Details {
     id=#3
     possessed_class=Overlay
     tracks [{
       type=MovieScene2DTransformTrack
       sections [{
         type=MovieScene2DTransformSection
         end_frame=16
         start_frame=0
         channels [
          Scale.X {type=MovieSceneScriptingFloatChannel, keys [[0,1],[16,1.2]]}
          Scale.Y {type=MovieSceneScriptingFloatChannel, keys [[0,1],[16,1.2]]}
         ]
        }]
      }]
    }
    Icon {
     id=#4
     possessed_class=Image
     tracks [{
       type=MovieSceneWidgetMaterialTrack
       sections [{
         type=MovieSceneParameterSection
         end_frame=16
         start_frame=0
         channels [
          Scale {type=MovieSceneScriptingFloatChannel}
          Green Channel ColorR {type=MovieSceneScriptingFloatChannel, keys [[0,0.0103],[16,0.1779]]}
          Green Channel ColorG {type=MovieSceneScriptingFloatChannel, keys [[0,0.0103],[16,0.1946]]}
          Green Channel ColorB {type=MovieSceneScriptingFloatChannel, keys [[0,0.0103],[16,0.2159]]}
          Green Channel ColorA {type=MovieSceneScriptingFloatChannel, const [1,2]}
         ]
        }]
      }]
    }
   ]
   display_rate {numerator=100}
  }
 }
]
graphs [
 ChangeKeybindLocation {nodes [
   ChangeKeybindLocation() {is=title K2Node_FunctionEntry}
   > SetHorizontalAlignment(
    self=SlotAsOverlaySlot(Widget=Get InputActionWidget() -> (InputActionWidget:Common Action Widget Object Reference))
    InHorizontalAlignment=Get InputActionWidgetHorizAlignment() -> (InputActionWidgetHorizAlignment:EHorizontalAlignment Enum=HAlign_Fill)
   )
   > n1 = Branch(Condition=Get bApplyOffsetAsTransformInsteadOfPaddingToIgnoreLayout() -> (bApplyOffsetAsTransformInsteadOfPaddingToIgnoreLayout:bool))
   n2 = Get InputActionWidget()
   n3 = Break Widget Transform!(WidgetTransform:"Widget Transform Structure (by ref)"=Get RenderTransform!(self:Widget Object Reference=@n2) -> (RenderTransform:Widget Transform Structure)) -> (Translation:Vector 2D Structure, Scale:Vector 2D Structure, Shear:Vector 2D Structure, Angle:float=~) {is=title K2Node_BreakStruct}
   n4 = Get KeyBind_Y_Offset() -> (KeyBind_Y_Offset:double)
   n5 = Get KeyBind_X_Offset() -> (KeyBind_X_Offset:double)
   > SetRenderTransform!(
    execute
    self:Widget Object Reference=@n2
    InTransform:Widget Transform Structure=Make Widget Transform!(
     Translation:Vector 2D Structure=MakeVector2D(X=@n5, Y=@n4)
     Scale:Vector 2D Structure=@n3.Scale
     Shear:Vector 2D Structure=@n3.Shear
     Angle:float=@n3.Angle
    ) -> (WidgetTransform:Widget Transform Structure) {is=title K2Node_MakeStruct}
   ) -> (then)
   n6 = SlotAsOverlaySlot(Widget=Get InputActionWidget())
   n7 = Break Margin!(Margin:"Margin Structure (by ref)"=Get Padding!(self:Overlay Slot Object Reference=@n6) -> (Padding:Margin Structure)) -> (Left:float=~, Top:float=~, Right:float=~, Bottom:float=~) {is=title K2Node_BreakStruct}
   SetPadding!(
    execute=@n1.else
    self:Overlay Slot Object Reference=@n6
    InPadding:Margin Structure=Make Margin(Left=Get KeyBind_X_Offset(), Top=Get KeyBind_Y_Offset(), Right=@n7.Right, Bottom=@n7.Bottom)
   ) -> (then)
  ]}
 EventGraph {nodes [
   OnTileReset() {is=title K2Node_CustomEvent}
   > SetIsInteractionEnabled!(execute, self:Common Button Base Object Reference, bInIsInteractionEnabled:bool=false) -> (then)
   > Check and Stop Animation!(execute:Exec, Animation:Widget Animation Object Reference=Get OnHover() -> (OnHover:Widget Animation Object Reference)) -> (then:Exec) {is=title K2Node_MacroInstance}
   > SetVisibility(self=Get Overlay_Button() -> (Overlay_Button:Overlay Object Reference), InVisibility=Collapsed)
   OnTileFocused() {is=title K2Node_CustomEvent}
   > SetIsInteractionEnabled(bInIsInteractionEnabled=true)
   > SetVisibility(self=Get Overlay_Button(), InVisibility=SelfHitTestInvisible)
   > PlayAnimationForward(InAnimation=Get OnFocus() -> (OnFocus:Widget Animation Object Reference))
   OnTileLostFocus() {is=title K2Node_CustomEvent}
   > SetIsInteractionEnabled(bInIsInteractionEnabled=false)
   > PlayAnimationReverse(InAnimation=Get OnFocus())
   Event Construct()
   > n1 = Parent: Construct()
   SetImageSize(execute=[@n1.then, @n3.then])
   > ChangeKeybindLocation() {is=title K2Node_CallFunction}
   > SetIsInteractionEnabled(bInIsInteractionEnabled=false)
   > SetVisibility(
    self=Get DropShadow() -> (DropShadow:Image Object Reference)
    InVisibility=Select(Option 0=Collapsed, Option 1=SelfHitTestInvisible, Index=Get bDisplayIconDropShadow() -> (bDisplayIconDropShadow:bool))
   )
   n2 = Event PreConstruct()
   > n3 = Branch(Condition=@n2.IsDesignTime)
   Event On Hovered()
   > Branch(Condition=Get bAllowDetailsButtonHoverAnim() -> (bAllowDetailsButtonHoverAnim:bool))
   n4 = Get OnHover()
   > PlayAnimationForward(InAnimation=@n4)
   Event On Unhovered()
   > Branch(Condition=Get bAllowDetailsButtonHoverAnim())
   > PlayAnimationReverse(InAnimation=@n4)
  ]}
 SetImageSize {nodes [
   SetImageSize() {is=title K2Node_FunctionEntry}
   > SetDesiredSizeOverride!(
    execute
    self:Image Object Reference=Get Icon() -> (Icon:Image Object Reference)
    DesiredSize:Vector 2D Structure=Get ImageSize() -> (ImageSize:Vector 2D Structure)
   ) -> (then)
  ]}
 SnapAnimationToEnd {nodes [
   n1 = SnapAnimationToEnd!() -> (then, InAnimation:Widget Animation Object Reference, PlayMode:EUMGSequencePlayMode Enum=Forward) {is=title K2Node_FunctionEntry}
   > PlayAnimation(InAnimation=@n1, StartAtTime=GetEndTime(self=@n1), PlayMode=@n1.PlayMode)
  ]}
]
variables [
 bAllowDetailsButtonHoverAnim {type=boolean, default=true}
 bApplyOffsetAsTransformInsteadOfPaddingToIgnoreLayout {type=boolean, default=true}
 bDisplayIconDropShadow {type=boolean}
 ImageSize {type=Vector2D, default=<64 64>}
 InputActionWidgetHorizAlignment {type=EHorizontalAlignment, default=HAlign_Center}
 KeyBind_X_Offset {type=number}
 KeyBind_Y_Offset {type=number, default=16}
]
widgets {
 info {inheritedWidgetCount=1, parentClass=&CommonButtonBase, rootWidgetClass=&Overlay, widgetCount=5}
 widgets [
  Overlay_Button {widgetClassPath=&Overlay, bIsVariable=true, widget=Overlay_Button}
  Overlay_Details {widgetClassPath=&Overlay, parent=Overlay_Button, widget=Overlay_Details, slot {class=OverlaySlot}}
  DropShadow {
   widgetClassPath=&Image
   bIsVariable=true
   parent=Overlay_Details
   widget=DropShadow
   properties {
    visibility=SelfHitTestInvisible
    brush {
     bIsDynamicallyLoaded=false
     drawAs=Image
     imageType=NoImage
     mirroring=NoMirror
     resourceName=None
     resourceObject=/FortniteCoreUI/UI/DiscoveryShared/Art/MI_UI_Icon_Details_DropShadow
     tiling=NoTile
     imageSize=<32 32>
     margin=<ltrb 0 0 0 0>
     outlineSettings {
      bUseBrushTransparency=false
      roundingType=HalfHeightRadius
      width=0
      color {colorUseRule=UseColor_Specified, specifiedColor=<rgba 0 0 0 0>}
      cornerRadii=<0 0 0 0>
     }
     tintColor {colorUseRule=UseColor_Specified, specifiedColor=<rgba 1 1 1 1>}
     uVRegion {bIsValid=false, max=<0 0>, min=<0 0>}
    }
   }
   slot {class=OverlaySlot, properties {horizontalAlignment=HAlign_Fill, verticalAlignment=VAlign_Fill}}
  }
  Icon {
   widgetClassPath=&Image
   bIsVariable=true
   parent=Overlay_Details
   widget=Icon
   properties {
    visibility=SelfHitTestInvisible
    brush {
     bIsDynamicallyLoaded=false
     drawAs=Image
     imageType=NoImage
     mirroring=NoMirror
     resourceName=None
     resourceObject=/FortniteCoreUI/UI/DiscoveryShared/Art/MI_UI_Icon_Details
     tiling=NoTile
     imageSize=<32 32>
     margin=<ltrb 0 0 0 0>
     outlineSettings {
      bUseBrushTransparency=false
      roundingType=HalfHeightRadius
      width=0
      color {colorUseRule=UseColor_Specified, specifiedColor=<rgba 0 0 0 0>}
      cornerRadii=<0 0 0 0>
     }
     tintColor {colorUseRule=UseColor_Specified, specifiedColor=<rgba 1 1 1 1>}
     uVRegion {bIsValid=false, max=<0 0>, min=<0 0>}
    }
   }
   slot {class=OverlaySlot, properties {horizontalAlignment=HAlign_Fill, verticalAlignment=VAlign_Fill}}
  }
  InputActionWidget {
   widgetClassPath=&CommonActionWidget
   bInherited=true
   bIsVariable=true
   parent=Overlay_Button
   widget=InputActionWidget
   properties {
    visibility=SelfHitTestInvisible
    iconRimBrush {
     bIsDynamicallyLoaded=false
     drawAs=Image
     imageType=NoImage
     mirroring=NoMirror
     resourceName=None
     resourceObject=/FortniteCoreUI/UI/DiscoveryShared/Art/MI_UI_Circle_WhiteKeybindBorder_SemiTransparent
     tiling=NoTile
     imageSize=<38 38>
     margin=<ltrb 0 0 0 0>
     outlineSettings {
      bUseBrushTransparency=false
      roundingType=HalfHeightRadius
      width=0
      color {colorUseRule=UseColor_Specified, specifiedColor=<rgba 0 0 0 0>}
      cornerRadii=<0 0 0 0>
     }
     tintColor {colorUseRule=UseColor_Specified, specifiedColor=<rgba 0 0 0 1>}
     uVRegion {bIsValid=false, max=<0 0>, min=<0 0>}
    }
   }
   slot {class=OverlaySlot, properties {horizontalAlignment=HAlign_Center, verticalAlignment=VAlign_Bottom, padding=<ltrb 4 4 4 4>}}
  }
 ]
}
